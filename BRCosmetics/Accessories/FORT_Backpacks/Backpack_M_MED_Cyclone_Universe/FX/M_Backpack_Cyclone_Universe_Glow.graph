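# Translucent unlit particle sprite material for the glow on the Cyclone Universe backpack, pulsing with audio.
# Particle color passes through an emissive adjustment function, while particle alpha scales a soft radial sphere mask.
# Emote high and low audio plus the Jerky event's kick-stem collection parameter are scaled, shaped by a power and clamped to modulate that mask.
# Depth fade softens intersections, except on the mobile shading path.
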Material /BRCosmetics/Accessories/FORT_Backpacks/Backpack_M_MED_Cyclone_Universe/FX/M_Backpack_Cyclone_Universe_Glow
graph {
 properties {
  bUsedWithParticleSprites=true
  blendMode=BLEND_Translucent
  pixelDepthOffsetMode=PDOM_Legacy
  shadingModel=MSM_Unlit
  usedShadingModels=Unlit
 }
 nodes [
  n1 = ParticleColor()
  MaterialFunctionCall(Emissive=@n1.RGB) {materialFunction=&MF_FX_EmissiveAdjustments}
  n2 = Constant() {r=0.5}
  n3 = Power(Base=SphereMask(A=TextureCoordinate(), B=@n2, Radius=@n2, Hardness=Constant()))
  n4 = CollectionParameter() {collection=&FortniteMaterialParameters, parameterName=Emote_AudioHigh}
  n5 = CollectionParameter() {collection=&FortniteMaterialParameters, parameterName=Emote_AudioLow}
  n6 = Add(A=CollectionParameter() {collection=/Game/Athena/Events/Jerky/MPC_JerkyCollection, parameterName=TrackStemKick}, B=Add(A=@n4, B=@n5))
  n7 = ScalarParameter() {defaultValue=3, parameterName=AudioMult}
  n8 = Add(A=Power(Base=Multiply(A=@n6, B=@n7), Exp=ScalarParameter() {defaultValue=1.5, parameterName=AudioPower})) {constB=0}
  n9 = ScalarParameter() {defaultValue=1.7, parameterName=AudioMax}
  n10 = Constant() {r=12}
  Multiply(
   A=Multiply(A=Multiply(A=@n1.A, B=@n3), B=Clamp(None=@n8, Max=@n9))
   B=ShadingPathSwitch(Default=DepthFade(FadeDistance=@n10), Mobile=Constant() {r=1}) {desc=Migrated from FeatureLevelSwitch}
  )
 ]
}
