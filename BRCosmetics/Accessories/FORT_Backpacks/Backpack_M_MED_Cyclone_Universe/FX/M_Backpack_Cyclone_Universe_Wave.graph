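# Translucent, unlit mesh-particle material for the Cyclone Universe backpack's wave effect, a ground-ring texture that reacts to audio.
# Jerky event kick and emote high/low audio parameters are summed, scaled and powered to drive opacity and refraction strength.
# A dynamic parameter rotates the ring mask and scales alpha, with optional depth fade on higher feature levels.
# Emissive color switches between particle color, time-based hue shift and a panned rainbow gradient remapped to HDR.
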Material /BRCosmetics/Accessories/FORT_Backpacks/Backpack_M_MED_Cyclone_Universe/FX/M_Backpack_Cyclone_Universe_Wave
graph {
 properties {
  bUsedWithMeshParticles=true
  blendMode=BLEND_Translucent
  pixelDepthOffsetMode=PDOM_Legacy
  refractionMethod=RM_IndexOfRefraction
  shadingModel=MSM_Unlit
  twoSided=true
  usedShadingModels=Unlit
 }
 nodes [
  n1 = ParticleColor()
  n2 = DynamicParameter() {paramNames [ScaleAlphaOnly, Param2, RotationInitial, RotationTime]}
  n3 = TextureSample(UVs=Rotator(Time=Add(A=@n2.RotationInitial, B=@n2.RotationInitial))) {samplerSource=SSM_Clamp_WorldGroupSettings, texture=/Game/Effects/Fort_Effects/Textures/Outlander/T_InTheZone_GroundRing_01}
  n4 = CollectionParameter() {collection=/Game/Athena/Events/Jerky/MPC_JerkyCollection, parameterName=TrackStemKick}
  n5 = CollectionParameter() {collection=&FortniteMaterialParameters, parameterName=Emote_AudioHigh}
  n6 = CollectionParameter() {collection=&FortniteMaterialParameters, parameterName=Emote_AudioLow}
  n7 = Add(A=@n5, B=@n6)
  n8 = Power(
   Base=Multiply(A=Add(A=@n4, B=@n7), B=ScalarParameter() {defaultValue=3, parameterName=AudioMult})
   Exp=ScalarParameter() {defaultValue=1.5, parameterName=AudioPower}
  )
  n9 = Multiply(A=@n8, B=ScalarParameter() {defaultValue=4, parameterName=RefractionIntensity})
  LinearInterpolate(B=ScalarParameter() {defaultValue=1.08, parameterName=RefractionIndex}, Alpha=Multiply(A=@n1.A, B=Multiply(A=@n3.G, B=@n9))) {constA=1, constB=1.1}
  n10 = ComponentMask(None=TextureCoordinate() {uTiling=0, vTiling=0}) {g=true}
  n11 = Multiply(
   A=TextureSample(UVs=Panner(Coordinate=@n10) {speedX=1.6667}) {texture=/Game/Global/Textures/Default/Masks_and_Gradients/T_Grad_Rainbow}.RGB
   B=ParticleColor().RGB
  )
  n12 = Clamp(None=@n11)
  n13 = ScalarParameter() {defaultValue=1, parameterName=Speed}
  n14 = Time()
  n15 = Multiply(
   A=MaterialFunctionCall(Hue Shift Percentage=Multiply(A=@n13, B=@n14), Texture=VectorParameter() {parameterName=StartColor, defaultValue=<rgba 1 0 0 1>}.RGB) {materialFunction=&HueShift}.Result
   B=@n1.RGB
  )
  MaterialFunctionCall(Emissive=StaticSwitchParameter(
    True=MaterialFunctionCall(LDR Input=@n12, HDR Tint=@n11) {materialFunction=&DeriveHDRfromLDR}.RemappedRange
    False=StaticSwitchParameter(True=@n15, False=@n1.RGB) {parameterName=UseRainbowColor}
   ) {parameterName=UseMellowRainbowColors}) {materialFunction=&MF_FX_EmissiveAdjustments}
  Reroute(None=TextureSample() {samplerType=SAMPLERTYPE_Normal, texture=/Game/Environments/World/Backgrounds/Universal/TechContent/Textures/T_Spherenormal_01}.RGB)
  n16 = Multiply(A=Multiply(A=@n2.ScaleAlphaOnly, B=Multiply(A=@n1.A, B=@n8)), B=@n3.G)
  n17 = DepthFade(Opacity=@n16, FadeDistance=ScalarParameter() {defaultValue=100, parameterName=DepthFadeDistance})
  StaticSwitchParameter(True=FeatureLevelSwitch(Default=@n17, ES2=@n16, ES3_1=@n16, SM4_REMOVED=@n17), False=@n16) {parameterName=UseDepthFade}
 ]
}
