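# Material function for the Cyclone Universe backpack that overrides incoming attributes to make an emote-audio-reactive planet and speaker look.
# Emote low/high audio modulates textured emissive and an inner glow built from distance gradients, fresnel, clouds and scrolling radial waves with a derivative-based seam fix.
# The same audio drives speaker wiggle, push and forward WPO, with PreviousFrameSwitch faking motion blur.
# Jerky track parameters and the RaveNinja idle/threshold chain are unconnected.
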MaterialFunction /BRCosmetics/Accessories/FORT_Backpacks/Backpack_M_MED_Cyclone_Universe/Material/MF_CycloneUniverseBackpack
graph {nodes [
  n1 = CollectionParameter() {collection=&FortniteMaterialParameters, parameterName=Emote_AudioLow}
  n2 = CollectionParameter() {collection=&FortniteMaterialParameters, parameterName=Emote_AudioHigh}
  n3 = Add(A=@n1, B=@n2)
  n4 = OneMinus(None=SphereMask(
    A=@n3
    B=ScalarParameter() {defaultValue=0.01, group=Raveninja, parameterName=RN_EmoteThreshold}
    Radius=Constant() {r=0.1}
    Hardness=Constant() {r=1}
   ))
  Clamp(None=@n4)
  CollectionParameter() {collection=/Game/Athena/Events/Jerky/MPC_JerkyCollection, parameterName=TrackBeatEnvelope}
  CollectionParameter() {collection=&MPC_JerkyCollection, parameterName=TrackLoudness}
  CollectionParameter() {collection=&MPC_JerkyCollection, parameterName=TrackStemKick}
  CollectionParameter() {collection=&MPC_JerkyCollection, parameterName=TrackStemSnare}
  CollectionParameter() {collection=&MPC_JerkyCollection, parameterName=TrackStemHats}
  CollectionParameter() {collection=&MPC_JerkyCollection, parameterName=TrackSpectrum}
  n5 = VectorParameter() {group=Planet, parameterName=Planet_BaseColor, defaultValue=<rgba 0.1 0.1 0.1 1>}
  n6 = TextureSampleParameter2D() {group=Planet, parameterName=BaseColor, texture=/Engine/EngineResources/Black_Low.Black}
  n7 = ComponentMask(None=VectorParameter() {group=Planet, parameterName=Planet_MonochromeScale, defaultValue=<rgba 0.4167 0.75 1.25 1>}.RGB) {b=true, g=true, r=true}
  n8 = Multiply(
   A=MaterialFunctionCall(X=@n6.R, Y=@n6.R, Z=@n6.R) {materialFunction=&MakeFloat3}.Result
   B=@n7
  )
  n9 = ScalarParameter() {parameterName=Planet_UseMonochrome}
  n10 = TextureSampleParameter2D() {group=Planet, parameterName=Emissive, texture=&Black}
  n11 = MaterialFunctionCall(X=@n10.R, Y=@n10.R, Z=@n10.R) {materialFunction=&MakeFloat3}
  n12 = CollectionParameter() {collection=&FortniteMaterialParameters, parameterName=Emote_AudioLow}
  n13 = Multiply(A=Multiply(A=@n12) {constB=4}) {constB=3}
  n14 = CollectionParameter() {collection=&FortniteMaterialParameters, parameterName=Emote_AudioHigh}
  n15 = Multiply(A=Add(A=@n14, B=@n12)) {constB=2}
  n16 = Multiply(A=@n15, B=Multiply(A=@n14) {constB=4})
  n17 = ScalarParameter() {defaultValue=1, group=Planet, parameterName=Planet_TextureEmissive_Scale}
  n18 = Multiply(A=Time()) {constB=0.25}
  n19 = Multiply(A=Sine(None=Frac(None=@n18))) {constB=0.5}
  n20 = Multiply(A=LinearInterpolate(A=@n10.RGB, B=Multiply(A=@n11.Result, B=@n7), Alpha=@n9), B=Add(A=Add(A=@n13, B=@n16), B=Add(A=@n17, B=@n19)))
  n21 = ScalarParameter() {defaultValue=1, group=Planet, parameterName=Planet_TextureEmissive_FinalScale}
  n22 = MaterialFunctionCall(Vector 3=MaterialFunctionCall() {materialFunction=&LocalPosition}."Local Position (Excluding Offsets)") {materialFunction=/Engine/Functions/Engine_MaterialFunctions02/Utility/VectorLength}
  n23 = ScalarParameter() {defaultValue=1, group=Planet, parameterName=Planet_InnerGlow_Scale, sliderMax=1}
  n24 = Power(Base=Saturate(None=Multiply(A=@n22.V3 Length, B=@n23) {constB=0.0005})) {constExponent=0.75}
  n25 = LinearInterpolate(
   A=VectorParameter() {group=Planet, parameterName=Planet_InnerColor_B, defaultValue=<rgba 1.3333 0 2 1>}.RGB
   B=VectorParameter() {group=Planet, parameterName=Planet_InnerColor_B_OnBeat, defaultValue=<rgba 1.3333 0 2 1>}.RGB
   Alpha=@n16
  )
  n26 = Saturate(None=Multiply(A=Subtract(A=@n22.V3 Length) {constB=1100}) {constB=0.005})
  n27 = LinearInterpolate(
   A=LinearInterpolate(
    A=VectorParameter() {group=Planet, parameterName=Planet_InnerColor_A, defaultValue=<rgba 1.4785 0.1211 1.75 1>}.RGB
    B=VectorParameter() {group=Planet, parameterName=Planet_InnerColor_A_OnBeat, defaultValue=<rgba 1.4785 0.1211 1.75 1>}.RGB
    Alpha=@n16
   )
   B=LinearInterpolate(
    A=VectorParameter() {group=Planet, parameterName=Speaker_InnerColor_A, defaultValue=<rgba 1.4785 0.1211 1.25 1>}.RGB
    B=VectorParameter() {group=Planet, parameterName=Speaker_InnerColor_A_OnBeat, defaultValue=<rgba 2 0.1638 1.6909 1>}.RGB
    Alpha=@n16
   )
   Alpha=@n26
  )
  n28 = Multiply(
   A=Add(A=@n22.V3 Length, B=ScalarParameter() {defaultValue=32, group=Planet, parameterName=Planet_InnerColor_Bias})
   B=ScalarParameter() {defaultValue=1, group=Planet, parameterName=Planet_InnerColor_Scale}
  ) {constB=0.0005}
  n29 = Power(Base=OneMinus(None=Saturate(None=@n28)), Exp=ScalarParameter() {defaultValue=1, group=Planet, parameterName=Planet_InnerColor_Power}) {constExponent=0.75}
  n30 = PixelNormalWS()
  n31 = CameraVectorWS()
  n32 = Saturate(None=Power(Base=DotProduct(A=@n30, B=@n31), Exp=ScalarParameter() {defaultValue=4, group=Planet, parameterName=Planet_InnerFresnel_Power}) {constExponent=4})
  n33 = Multiply(A=@n32, B=ScalarParameter() {defaultValue=0.75, group=Planet, parameterName=Planet_InnerFresnel_Scale}) {constB=0.75}
  n34 = ScalarParameter() {defaultValue=0.25, group=Planet, parameterName=Planet_InnerFresnel_Bias}
  n35 = Multiply(A=@n24, B=Multiply(A=LinearInterpolate(A=@n25, B=@n27, Alpha=@n29), B=Add(A=@n33, B=@n34) {constB=0.25}))
  n36 = MaterialFunctionCall!(Center Point World Position=MaterialFunctionCall() {materialFunction=&ComponentPivotLocation}.Result) {materialFunction=/Engine/Functions/Engine_MaterialFunctions02/UVs/WorldSpaceAlignedScreenCoordinates}
  n37 = ScalarParameter() {defaultValue=4, group=Planet, parameterName=Planet_InnerCloud_UVScale}
  n38 = TextureSample(UVs=Rotator(Coordinate=Multiply(A=@n36.Offset and Scaled UVS, B=@n37) {constB=0.05}) {speed=0.1}) {texture=/Game/Athena/SafeZone/S11/T_TilingClouds_D}
  n39 = Constant() {r=1}
  n40 = Add(A=Subtract(A=Multiply(A=@n16) {constB=0.01}) {constB=0.005})
  n41 = Power(
   Base=MaterialFunctionCall!(Vector or UVs=@n36.Offset and Scaled UVS) {materialFunction=/Game/Athena/SafeZone/S11/MF_VectorToRadialValueFast}.Result
   Exp=Multiply(A=Constant2Vector() {g=1.3, r=1}, B=AppendVector(A=@n39, B=@n40))
  ) {constExponent=1.25}
  n42 = ComponentMask(None=@n41) {r=true}
  MaterialExpressionFrac_8 = Frac(None=@n42)
  MaterialExpressionSubtract_4 = Subtract(A=Frac(None=Add(A=@n42) {constB=0.5})) {constB=0.5}
  n43 = Custom!(u0=@MaterialExpressionFrac_8, u1=@MaterialExpressionSubtract_4) {
   code="return ((fwidth(u0) < fwidth(u1)-0.001)? u0:u1);"
   description=FixPerPixelCylidricalUV
   outputType=CMOT_Float1
   inputs [
    {
     inputName=u0
     input {expression=&self:MaterialExpressionFrac_8}
    }
    {
     inputName=u1
     input {expression=&self:MaterialExpressionSubtract_4}
    }
   ]
  }
  n44 = ComponentMask(None=@n43) {r=true}
  n45 = Multiply(A=@n44, B=ScalarParameter() {defaultValue=50, group=Planet, parameterName=Planet_WavesLayer1_U_Tiling}) {constB=35}
  n46 = ComponentMask(None=@n41) {g=true}
  n47 = Time()
  n48 = ScalarParameter() {defaultValue=-1, parameterName=Planet_WavesLayer1_V_Speed}
  n49 = Add(
   A=Multiply(A=@n46, B=ScalarParameter() {defaultValue=20, group=Planet, parameterName=Planet_WavesLayer1_V_Tiling}) {constB=0.2}
   B=Frac(None=Multiply(A=@n47, B=@n48) {constB=-1})
  )
  n50 = ScalarParameter() {defaultValue=10, group=Planet, parameterName=Planet_WavesLayer2_U_ScaleOnBeat}
  n51 = Multiply(
   A=@n44
   B=Add(A=ScalarParameter() {defaultValue=25, group=Planet, parameterName=Planet_WavesLayer2_U_Tiling}, B=Multiply(A=@n15, B=@n50) {constB=20})
  ) {constB=15}
  n52 = Time()
  n53 = ScalarParameter() {defaultValue=-1, parameterName=Planet_WavesLayer2_V_RotationSpeed}
  n54 = Add(A=@n44, B=Frac(None=Multiply(A=@n52, B=@n53) {constB=-1}))
  n55 = Multiply(A=@n46, B=ScalarParameter() {defaultValue=15, group=Planet, parameterName=Planet_WavesLayer2_V_Tiling}) {constB=0.15}
  n56 = Frac(None=Multiply(A=@n47, B=ScalarParameter() {defaultValue=-1.5, parameterName=Planet_WavesLayer2_V_Speed}) {constB=-1.5})
  n57 = TextureSampleParameter2D(UVs=AppendVector(A=Add(A=@n51, B=@n54), B=Add(A=@n55, B=@n56))) {
   parameterName=Planet_InnerTexture_B
   samplerType=SAMPLERTYPE_LinearColor
   texture=/Game/Environments/Apollo/Sets/Jerky/Materials/T_JerkyFlowmap
  }
  n58 = AppendVector(A=@n57.R, B=@n57.R)
  n59 = ComponentMask(None=VectorParameter() {group=Planet, parameterName=Planet_WavesUVDeformStrength, defaultValue=<rgba 0.3 0.4 0 1>}.RGB) {g=true, r=true}
  n60 = TextureSampleParameter2D(UVs=Add(A=AppendVector(A=@n45, B=@n49), B=Multiply(A=@n58, B=@n59) {constB=0.25})) {
   group=Planet
   parameterName=Planet_InnerTexture_A
   samplerType=SAMPLERTYPE_LinearColor
   texture=&T_JerkyFlowmap
  }
  n61 = LinearInterpolate(A=Multiply(A=OneMinus(None=@n38.R), B=Multiply(A=@n60.R, B=@n57.G)), Alpha=@n26) {constA=1}
  n62 = Multiply(
   A=Multiply(A=@n24, B=Multiply(A=@n32, B=@n61))
   B=LinearInterpolate(
    A=ScalarParameter() {defaultValue=2, group=Planet, parameterName=Planet_InnerWaves_Intensity}
    B=ScalarParameter() {defaultValue=2, group=Planet, parameterName=Planet_InnerWaves_Intensity_OnBeat}
    Alpha=@n15
   )
  ) {constB=2}
  n63 = Add(A=@n35, B=Saturate(None=Subtract(A=@n35, B=@n62)))
  n64 = ScalarParameter() {defaultValue=1, group=Planet, parameterName=Planet_InnerEmissive_FinalScale}
  n65 = Multiply(A=Add(A=@n15) {constB=0.25}, B=Add(A=@n16, B=@n15))
  n66 = Subtract(A=@n22.V3 Length)
  n67 = OneMinus(None=Saturate(None=Multiply(A=@n66) {constB=2}))
  n68 = Multiply(A=@n67, B=@n65)
  n69 = Multiply(A=@n68, B=ScalarParameter() {defaultValue=3, group=Speaker, parameterName=Speaker_WPO_WiggleIntensity_OnBeat}) {constB=750}
  n70 = Multiply(A=Time(), B=Add(A=ScalarParameter() {defaultValue=1, group=Wiggle, parameterName=WiggleSpeedMultiplier}, B=@n69))
  n71 = Add(A=@n70) {constB=0.25}
  n72 = Add(A=@n70) {constB=0.1155}
  n73 = Add(A=@n70) {constB=-0.6541}
  n74 = MaterialFunctionCall(
   R=LinearInterpolate(A=Sine(None=@n71) {period=0.665}, B=Sine(None=@n71) {period=1.4857}, Alpha=Sine(None=@n71) {period=0.8})
   G=LinearInterpolate(A=Sine(None=@n72) {period=0.75}, B=Sine(None=@n72) {period=1.2}, Alpha=Sine(None=@n72) {period=0.9987})
   B=LinearInterpolate(A=Sine(None=@n73) {period=0.5}, B=Sine(None=@n73) {period=1.4857}, Alpha=Sine(None=@n73))
  ) {materialFunction=&AppendMany}
  n75 = Add(A=ScalarParameter() {group=Wiggle, parameterName=WiggleIntensity}, B=@n69)
  n76 = ScalarParameter() {group=Wiggle, parameterName=WiggleSpeedLerp}
  n77 = Multiply(
   A=Power(Base=@n68) {constExponent=0.5}
   B=ScalarParameter() {defaultValue=3, group=Speaker, parameterName=Speaker_WPO_WiggleSpeedLerp_OnBeat}
  )
  n78 = MaterialFunctionCall() {materialFunction=&ComponentPivotLocation}
  n79 = VectorParameter() {parameterName=PivotOffset, defaultValue=<rgba -1 -13 -1 1>}
  n80 = Normalize(VectorInput=Subtract(A=WorldPosition() {worldPositionShaderOffset=WPT_ExcludeAllShaderOffsets}.XYZ, B=Add(A=@n78.Result, B=@n79.RGB)))
  n81 = LinearInterpolate(
   A=ScalarParameter() {group=Speaker, parameterName=Speaker_WPO_Push}
   B=ScalarParameter() {defaultValue=3, group=Speaker, parameterName=Speaker_WPO_Push_OnBeat}
   Alpha=@n65
  )
  n82 = Add(A=LinearInterpolate(B=Multiply(A=@n74.RGB, B=@n75), Alpha=Add(A=@n76, B=@n77)) {constB=0}, B=Multiply(A=@n67, B=Multiply(A=@n80, B=@n81)))
  n83 = PreviousFrameSwitch(Current Frame=@n82, Previous Frame=Multiply(A=@n82, B=ScalarParameter() {defaultValue=3, group=Fake Blur, parameterName=FakeBlurMultiplier}))
  n84 = ScalarParameter() {group=Fake Blur, parameterName=FakeBlurLerp}
  n85 = Multiply(A=@n15, B=ScalarParameter() {defaultValue=1, group=Speaker, parameterName=Speaker_WPO_ForwardMovement_Intensity})
  n86 = MaterialFunctionCall() {materialFunction=&LocalPosition}
  n87 = ScalarParameter() {defaultValue=4, group=Speaker, parameterName=Speaker_WPO_ForwardMovement_OnBeat}
  n88 = Multiply(A=@n87) {constB=0.0025}
  n89 = MaterialFunctionCall(Float3=Subtract(A=Multiply(A=@n86.Local Position, B=@n88), B=@n86.Local Position)) {materialFunction=&BreakOutFloat3Components}
  n90 = Multiply(A=@n15, B=@n87) {constB=5000}
  n91 = MaterialFunctionCall(Float3=@n86."Local Position (Excluding Offsets)") {materialFunction=&BreakOutFloat3Components}
  n92 = Distance(A=AppendVector(A=@n91.R, B=@n91.R), B=Constant2Vector())
  n93 = Saturate(None=Multiply(A=Subtract(A=@n92) {constB=500}) {constB=0.0015})
  n94 = Multiply(
   A=Transform(None=MaterialFunctionCall(X=@n89.R, Y=@n90, Z=@n89.R) {materialFunction=&MakeFloat3}.Result) {transformSourceType=TRANSFORMSOURCE_Local}
   B=Power(Base=OneMinus(None=@n93))
  )
  n95 = TextureSampleParameter2D() {group=Planet, parameterName=Normal, samplerType=SAMPLERTYPE_Normal, texture=&FlatNormalBlue}
  n96 = SetMaterialAttributes!(
   MaterialAttributes=GetMaterialAttributes(None=FunctionInput() {inputName=Material, inputType=FunctionInput_MaterialAttributes}).MaterialAttributes
   BaseColor=Add(A=ComponentMask(None=@n5.RGB) {b=true, g=true, r=true}, B=LinearInterpolate(A=@n6.RGB, B=@n8, Alpha=@n9))
   Roughness=ScalarParameter() {defaultValue=0.8, parameterName=Planet_Roughness}
   EmissiveColor=Add(A=Multiply(A=@n20, B=@n21), B=Multiply(A=@n63, B=@n64))
   Normal=@n95.RGB
   WorldPositionOffset=Add(A=LinearInterpolate(A=@n82, B=@n83, Alpha=@n84), B=Multiply(A=@n85, B=@n94))
  ) {attributeSetTypes [
    #1
    #2
    #3
    #4
    #5
   ]}
  FunctionOutput(None=@n96)
  n97 = Time()
  n98 = ScalarParameter() {defaultValue=0.2, group=RaveNinja, parameterName=RN_DefaultSineWaveSpeed}
  n99 = ConstantBiasScale(None=Sine(None=Multiply(A=@n97, B=@n98)))
  n100 = ScalarParameter() {defaultValue=1, group=RaveNinja, parameterName=RN_DefaultAmplitudeBias}
  n101 = Multiply(A=Add(A=@n97) {constB=0.25}, B=Constant() {r=0.2})
  n102 = Sine(None=Multiply(A=Add(A=@n97) {constB=0.5}) {constB=0.5})
  n103 = Sine(None=@n97)
  n104 = Add(A=Abs(None=Sine(None=@n101)), B=Add(A=Abs(None=@n102), B=Abs(None=@n103)))
  n105 = Ceil(None=@n3)
  LinearInterpolate(
   A=LinearInterpolate(A=Divide(A=@n99, B=@n100), B=Divide(A=@n104) {constA=3, constB=3}, Alpha=Clamp(None=@n105))
   B=OneMinus(None=@n3)
   Alpha=@n4
  )
  OneMinus(None=@n26)
  VertexNormalWS()
 ]}
