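# Material function for the Cyclone Universe backpack that makes its planet/speaker surface pulse and glow with music.
# Event audio, falling back to emote audio when quiet, drives world position offset away from Speaker_Location, masked by vertex blue, with an idle sine pulse during silence.
# UseWithTextures selects masked texture glow or procedural view-angle gradients with beat-reactive stripes from a panning radial flowmap, using a derivative-based seam fix.
# Jerky stem nodes are disconnected from the output.
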MaterialFunction /BRCosmetics/Accessories/FORT_Backpacks/Backpack_M_MED_Cyclone_Universe/Material/MF_CycloneUniverseBackpack_V2
graph {nodes [
  n1 = CollectionParameter() {collection=&FortniteMaterialParameters, parameterName=EventAudio_Amplitude}
  n2 = Subtract(A=Multiply(A=@n1) {constB=2})
  n3 = CollectionParameter() {collection=&FortniteMaterialParameters, parameterName=EventAudio_Spectrum}
  n4 = Power(Base=Add(A=Multiply(A=@n2) {constB=0.5}, B=ComponentMask(None=@n3) {g=true}))
  n5 = Subtract(A=Multiply(A=CollectionParameter() {collection=/Game/Athena/Events/Jerky/MPC_JerkyCollection, parameterName=TrackStemSnare}) {constB=2})
  n6 = Multiply(A=CollectionParameter() {collection=&MPC_JerkyCollection, parameterName=TrackStemHats}) {constB=2}
  Clamp(None=Add(A=Add(A=@n4, B=@n5), B=Subtract(A=@n6)))
  CollectionParameter() {collection=&MPC_JerkyCollection, parameterName=TrackBeatEnvelope}
  CollectionParameter() {collection=&MPC_JerkyCollection, parameterName=TrackStemKick}
  CollectionParameter() {collection=&MPC_JerkyCollection, parameterName=TrackStemKick}
  CollectionParameter() {collection=&MPC_JerkyCollection, parameterName=TrackStemSnare}
  CollectionParameter() {collection=&MPC_JerkyCollection, parameterName=TrackStemHats}
  CollectionParameter() {collection=&MPC_JerkyCollection, parameterName=TrackSpectrum}
  Constant3Vector()
  n7 = GetMaterialAttributes(None=FunctionInput() {inputName=Material, inputType=FunctionInput_MaterialAttributes}) {attributeGetTypes [#1, #2]}
  n8 = VectorParameter() {parameterName=Speaker_Location, defaultValue=<rgba 0 0 0 1>}
  n9 = Normalize(VectorInput=Subtract(A=WorldPosition().XYZ, B=@n8.RGB))
  n10 = ScalarParameter() {defaultValue=1.5, parameterName=CU_Beat_Offset_Amt}
  n11 = Time()
  n12 = Sine(None=Multiply(A=Add(A=@n11) {constB=0.25}, B=Constant() {r=0.2}))
  n13 = Add(A=@n11) {constB=0.5}
  n14 = Abs(None=Sine(None=Multiply(A=@n13) {constB=0.5}))
  n15 = Abs(None=Sine(None=@n11))
  n16 = Divide(A=Add(A=Abs(None=@n12), B=Add(A=@n14, B=@n15))) {constA=3, constB=3}
  n17 = CollectionParameter() {collection=&FortniteMaterialParameters, parameterName=Emote_AudioHigh}
  n18 = CollectionParameter() {collection=&FortniteMaterialParameters, parameterName=Emote_AudioLow}
  n19 = ScalarParameter() {defaultValue=0.01, group=Raveninja, parameterName=RN_EmoteThreshold}
  n20 = Constant() {r=0.1}
  n21 = Constant() {r=1}
  n22 = LinearInterpolate(A=Add(A=@n17, B=@n18), B=@n4, Alpha=OneMinus(None=SphereMask(A=@n4, B=@n19, Radius=@n20, Hardness=@n21)))
  n23 = Power(Base=@n22)
  n24 = OneMinus(None=SphereMask(
    A=@n22
    B=ScalarParameter() {defaultValue=0.01, group=Raveninja, parameterName=RN_EmoteThreshold}
    Radius=Constant() {r=0.1}
    Hardness=Constant() {r=1}
   ))
  n25 = Multiply(A=Multiply(A=@n10, B=LinearInterpolate(A=@n16, B=@n23, Alpha=@n24)), B=VertexColor().B)
  n26 = VectorParameter() {group=Planet, parameterName=Planet_InnerColor_B, defaultValue=<rgba 0.8073 0.0661 0.6825 1>}
  n27 = TextureSample() {
   samplerType=SAMPLERTYPE_Masks
   texture=/BRCosmetics/Accessories/FORT_Backpacks/Backpack_M_MED_Cyclone_Universe/Textures/T_M_MED_Cyclone_Universe_Backpack_Planet_FX
  }
  n28 = FunctionInput(Preview=@n27.RGB) {bUsePreviewValueAsDefault=true, inputName=FX_Mask}
  n29 = Multiply(A=ComponentMask(None=@n28) {r=true}, B=ScalarParameter() {defaultValue=1, parameterName=TextureEmissive_Strength_02})
  n30 = Multiply(
   A=FunctionInput(Preview=Constant() {r=1}) {bUsePreviewValueAsDefault=true, inputName=Gmap, inputType=FunctionInput_Scalar}
   B=ScalarParameter() {defaultValue=0.05, parameterName=GmapStrength}
  )
  n31 = Add(A=@n29, B=Multiply(A=ComponentMask(None=@n28) {g=true}, B=ScalarParameter() {defaultValue=1, parameterName=TextureEmissive_Strength}))
  n32 = ScalarParameter() {parameterName=UseGmapGlow}
  n33 = Multiply(A=@n26.RGB, B=LinearInterpolate(A=@n29, B=Add(A=@n30, B=@n31), Alpha=@n32))
  n34 = LinearInterpolate(A=Add(A=@n7.EmissiveColor, B=@n33), B=Add(A=@n7.EmissiveColor, B=@n33), Alpha=@n32)
  n35 = VectorParameter() {group=Planet, parameterName=Planet_InnerColor_A, defaultValue=<rgba 0.35 0 0.525 1>}
  n36 = DotProduct(A=CameraVectorWS(), B=PixelNormalWS())
  n37 = ScalarParameter() {parameterName=Radial_BeatMin}
  n38 = ScalarParameter() {defaultValue=0.75, parameterName=Radial_BeatMax}
  n39 = Power(Base=Multiply(A=@n22) {constB=2})
  n40 = Power(Base=Subtract(A=@n36, B=LinearInterpolate(A=@n37, B=@n38, Alpha=@n39))) {constExponent=4}
  n41 = VectorParameter() {group=Planet, parameterName=Planet_Stripes_Color, defaultValue=<rgba 1 0.1221 1 1>}
  n42 = LinearInterpolate(
   A=ScalarParameter() {defaultValue=2, parameterName=GlowStrengthMin}
   B=ScalarParameter() {defaultValue=5, parameterName=GlowStrengthMax}
   Alpha=@n39
  )
  n43 = MaterialFunctionCall!(Center Point World Position=@n8.RGB) {materialFunction=/Engine/Functions/Engine_MaterialFunctions02/UVs/WorldSpaceAlignedScreenCoordinates}
  n44 = MaterialFunctionCall!(Vector or UVs=@n43.Offset and Scaled UVS) {materialFunction=/Game/Athena/SafeZone/S11/MF_VectorToRadialValueFast}
  n45 = ComponentMask(None=@n44.Result) {r=true}
  MaterialExpressionFrac_1 = Frac(None=@n45)
  MaterialExpressionSubtract_11 = Subtract(A=Frac(None=Add(A=@n45) {constB=0.5})) {constB=0.5}
  n46 = Custom!(u0=@MaterialExpressionFrac_1, u1=@MaterialExpressionSubtract_11) {
   code="return ((fwidth(u0) < fwidth(u1)-0.001)? u0:u1);"
   description=FixPerPixelCylidricalUV
   outputType=CMOT_Float1
   inputs [
    {
     inputName=u0
     input {expression=&self:MaterialExpressionFrac_1}
    }
    {
     inputName=u1
     input {expression=&self:MaterialExpressionSubtract_11}
    }
   ]
  }
  n47 = Add(A=ComponentMask(None=@n46) {r=true}, B=Multiply(A=Time(), B=ScalarParameter() {defaultValue=0.2, parameterName=CU_RadialNoisePanSpeed}))
  n48 = ComponentMask(None=@n44.Result) {g=true}
  n49 = VectorParameter() {parameterName=CU_RadialNoise_01_Tiling, defaultValue=<rgba 24 0 0 1>}
  n50 = TextureSample(UVs=Multiply(A=AppendVector(A=@n47, B=@n48), B=ComponentMask(None=@n49.RGB) {g=true, r=true})) {
   samplerType=SAMPLERTYPE_LinearColor
   texture=/BRCosmetics/Accessories/FORT_Backpacks/Backpack_M_MED_Cyclone_Universe/Textures/T_JerkyFlowmap_Backpack
  }
  n51 = Subtract(
   A=Multiply(A=@n36, B=@n50.G)
   B=LinearInterpolate(A=ScalarParameter() {defaultValue=0.5, parameterName=Snare_beatMin}, B=ScalarParameter() {parameterName=Snare_beat_max}, Alpha=@n39)
  )
  n52 = Add(
   A=@n7.EmissiveColor
   B=LinearInterpolate(A=LinearInterpolate(A=@n35.RGB, B=@n26.RGB, Alpha=@n40), B=Multiply(A=@n41.RGB, B=@n42), Alpha=Clamp(None=@n51))
  )
  FunctionOutput(None=SetMaterialAttributes!(
    MaterialAttributes=@n7.MaterialAttributes
    WorldPositionOffset=Multiply(A=@n9, B=@n25)
    EmissiveColor=StaticSwitchParameter(True=@n34, False=@n52) {parameterName=UseWithTextures}
    BaseColor=StaticSwitchParameter(True=@n7.BaseColor, False=@n52) {parameterName=UseWithTextures}
   ) {attributeSetTypes [#3, #1, #2]})
  LinearInterpolate()
  MaterialFunctionCall() {materialFunction=&ObjectPivotPoint}
  Multiply(
   A=SphereMask(A=@n43.Offset UVs, B=Constant2Vector(), Radius=ScalarParameter() {defaultValue=6, parameterName=RadialGradientScale}, Hardness=Constant())
   B=@n10
  )
  OneMinus(None=@n36)
  PreSkinnedPosition()
  PreSkinnedPosition()
  Reroute(None=@n39)
  n53 = VectorParameter() {parameterName=CU_PivotOffset, defaultValue=<rgba 1 13 0 1>}
  Subtract(B=@n53.RGB)
  Subtract(
   A=MaterialFunctionCall() {materialFunction=&ComponentPivotLocation}.Result
   B=@n53.RGB
  )
  VectorParameter() {parameterName=FerrisWheel_Location, defaultValue=<rgba 0 0 0 1>}
  VectorParameter() {parameterName=Spinner_01_Location, defaultValue=<rgba 0 0 0 1>}
  VectorParameter() {parameterName=Spinner_02_Location, defaultValue=<rgba 0 0 0 1>}
  VertexNormalWS()
 ]}
