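# Material function for the Cyclone Universe backpack that adds animated amusement-ride glows and colors the coaster track.
# Vertex red isolates Ferris wheel and two spinners, lit by curve-atlas radial pulses around pre-skinned pivot offsets.
# Vertex green marks the track: sine-striped base color and a UV1-scrolling coaster glow, widened and brightened by Jerky event collection's TrackStemHats.
# Only EmissiveColor and BaseColor are modified; ride rotation, normal fix-up and snare-driven speed branches are unconnected.
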MaterialFunction /BRCosmetics/Accessories/FORT_Backpacks/Backpack_M_MED_Cyclone_Universe/Material/MF_CycloneUniverse_Emissives
graph {nodes [
  n1 = VertexColor()
  n2 = SphereMask(A=@n1.R, B=ScalarParameter() {defaultValue=0.25, parameterName=FerrisWheel_VertValue}, Radius=Constant() {r=0.1}, Hardness=Constant() {r=1})
  n3 = VectorParameter() {parameterName=FerrisWheelRotationOrientation, defaultValue=<rgba -0.565 -0.776 0.279 1>}
  n4 = Frac(None=Multiply(A=Time(), B=ScalarParameter() {defaultValue=0.05, parameterName=FerrisWheel_RideSpeed}) {constB=0.05})
  n5 = WorldPosition()
  n6 = RotateAboutAxis(
   NormalizedRotationAxis=@n3.RGB
   RotationAngle=@n4
   PivotPoint=VectorParameter() {parameterName=FerisWheelLocalOffset_WPO, defaultValue=<rgba 2.504 -8.306 12.743 1>}.RGB
   Position=@n5.XYZ
  )
  n7 = SphereMask(A=@n1.R, B=ScalarParameter() {defaultValue=0.5, parameterName=Spinner_01_VertValue}, Radius=Constant() {r=0.1}, Hardness=Constant() {r=1})
  n8 = VectorParameter() {parameterName=Spinner_A_RotationOrientation, defaultValue=<rgba -0.835 0.16 0.526 1>}
  n9 = Frac(None=Multiply(A=Time(), B=ScalarParameter() {defaultValue=0.1, parameterName=Spinner_A_RideSpeed}) {constB=0.05})
  n10 = RotateAboutAxis(
   NormalizedRotationAxis=@n8.RGB
   RotationAngle=@n9
   PivotPoint=VectorParameter() {parameterName=Spinner_01_LocalOffset_WPO, defaultValue=<rgba -8.641 -10.861 5.779 1>}.RGB
   Position=@n5.XYZ
  )
  n11 = SphereMask(A=@n1.R, B=ScalarParameter() {defaultValue=0.75, parameterName=Spinner_02_VertValue}, Radius=Constant() {r=0.1}, Hardness=Constant() {r=1})
  n12 = VectorParameter() {parameterName=Spinner_B_RotationOrientation, defaultValue=<rgba -0.653 -0.682 0.331 1>}
  n13 = Frac(None=Multiply(A=Time(), B=ScalarParameter() {defaultValue=0.15, parameterName=Spinner_B_RideSpeed}) {constB=0.05})
  n14 = VectorParameter() {parameterName=Spinner_02_LocalOffset_WPO, defaultValue=<rgba -6.795 -18.969 3.181 1>}
  Add(
   A=Add(A=Multiply(A=@n2, B=@n6), B=Multiply(A=@n7, B=@n10))
   B=Multiply(A=@n11, B=RotateAboutAxis(NormalizedRotationAxis=@n12.RGB, RotationAngle=@n13, PivotPoint=@n14.RGB, Position=@n5.XYZ))
  )
  n15 = TextureCoordinate()
  n16 = Sine(None=Multiply(A=ComponentMask(None=@n15) {g=true}, B=ScalarParameter() {defaultValue=15, parameterName=TrackStep Count}))
  Clamp(None=@n16)
  n17 = GetMaterialAttributes(None=FunctionInput() {inputName=Material, inputType=FunctionInput_MaterialAttributes}) {attributeGetTypes [#1, #2]}
  n18 = VectorParameter() {parameterName=CoasterGlowColor, defaultValue=<rgba 1 0.5119 0 1>}
  n19 = Time()
  n20 = ScalarParameter() {defaultValue=0.25, parameterName=CoasterSpeed}
  n21 = Add(A=ComponentMask(None=TextureCoordinate() {coordinateIndex=1}) {g=true}, B=Frac(None=Multiply(A=@n19, B=@n20)))
  n22 = Constant() {r=0.5}
  n23 = CollectionParameter() {collection=/Game/Athena/Events/Jerky/MPC_JerkyCollection, parameterName=TrackStemHats}
  n24 = LinearInterpolate(
   A=ScalarParameter() {defaultValue=0.05, parameterName=CoasterSpread}
   B=ScalarParameter() {defaultValue=0.1, parameterName=CoasterSpread}
   Alpha=@n23
  )
  n25 = ScalarParameter() {defaultValue=0.5, parameterName=Coasterhardness}
  n26 = Multiply(
   A=Multiply(A=@n18.RGB, B=SphereMask(A=@n21, B=@n22, Radius=@n24, Hardness=@n25))
   B=LinearInterpolate(
    A=ScalarParameter() {defaultValue=2, parameterName=CoasterGlowStrength}
    B=ScalarParameter() {defaultValue=5, parameterName=CoasterGlowStrengthMAx}
    Alpha=@n23
   )
  )
  n27 = SphereMask(A=@n1.G, B=Constant() {r=0.5}, Radius=Constant() {r=0.1}, Hardness=Constant() {r=1})
  n28 = VertexInterpolator(VS=PreSkinnedPosition())
  n29 = VectorParameter() {parameterName=FerisWheelLocalOffset, defaultValue=<rgba 2.504 -8.306 12.743 1>}
  n30 = VectorParameter() {parameterName=Backpack_GlobalOffset, defaultValue=<rgba 0 0 0 1>}
  n31 = Multiply(A=SphereMask(A=@n28.PS, B=Add(A=@n29.RGB, B=@n30.RGB), Radius=Constant() {r=3.5}, Hardness=Constant())) {constB=2}
  n32 = Multiply(A=Time(), B=ScalarParameter() {defaultValue=0.5, parameterName=RadialEmissive_PanSpeed})
  n33 = Distance(A=Frac(None=Add(A=@n31, B=@n32)), B=Constant())
  n34 = ScalarParameter() {defaultValue=0.5, parameterName=GlowRadius}
  n35 = Clamp(None=OneMinus(None=Divide(A=@n33, B=@n34)))
  n36 = CurveAtlasRowParameter(CurveTime=@n35) {
   atlas=&AthenaCurveAtlas01
   curve=/Game/Athena/LinearColorCurves/Curve_JellyfishPickaxe_Swing2
   defaultValue=100
  }
  n37 = Multiply(A=Multiply(A=Multiply(A=@n35, B=@n36), B=ScalarParameter() {defaultValue=2, parameterName=RidesGlowStrength}), B=@n2)
  n38 = VectorParameter() {parameterName=Spinner_01_LocalOffset, defaultValue=<rgba -8.641 -10.861 5.779 1>}
  n39 = Multiply(A=SphereMask(A=@n28.PS, B=Add(A=@n38.RGB, B=@n30.RGB), Radius=Constant() {r=3.5}, Hardness=Constant())) {constB=2}
  n40 = Multiply(A=Time(), B=ScalarParameter() {defaultValue=0.5, parameterName=RadialEmissive_PanSpeed})
  n41 = Distance(A=Frac(None=Add(A=@n39, B=@n40)), B=Constant())
  n42 = ScalarParameter() {defaultValue=0.5, parameterName=GlowRadius}
  n43 = Clamp(None=OneMinus(None=Divide(A=@n41, B=@n42)))
  n44 = CurveAtlasRowParameter(CurveTime=@n43) {
   atlas=&AthenaCurveAtlas01
   curve=&Curve_JellyfishPickaxe_Swing2
   defaultValue=100
  }
  n45 = Multiply(A=Multiply(A=Multiply(A=@n43, B=@n44), B=ScalarParameter() {defaultValue=5, parameterName=RidesGlowStrength}), B=@n7)
  n46 = VectorParameter() {parameterName=Spinner_02_LocalOffset, defaultValue=<rgba -6.795 -18.969 3.181 1>}
  n47 = Multiply(A=SphereMask(A=@n28.PS, B=Add(A=@n46.RGB, B=@n30.RGB), Radius=Constant() {r=3.5}, Hardness=Constant())) {constB=2}
  n48 = Multiply(A=Time(), B=ScalarParameter() {defaultValue=0.5, parameterName=RadialEmissive_PanSpeed})
  n49 = Distance(A=Frac(None=Add(A=@n47, B=@n48)), B=Constant())
  n50 = ScalarParameter() {defaultValue=0.5, parameterName=GlowRadius}
  n51 = Clamp(None=OneMinus(None=Divide(A=@n49, B=@n50)))
  n52 = CurveAtlasRowParameter(CurveTime=@n51) {
   atlas=&AthenaCurveAtlas01
   curve=&Curve_JellyfishPickaxe_Swing2
   defaultValue=100
  }
  n53 = Multiply(A=Multiply(A=@n51, B=@n52), B=ScalarParameter() {defaultValue=5, parameterName=RidesGlowStrength})
  n54 = Add(A=Multiply(A=Power(Base=@n26), B=@n27), B=Add(A=Add(A=@n37, B=@n45), B=Multiply(A=@n53, B=@n11)))
  n55 = Add(A=@n17.EmissiveColor, B=@n54)
  n56 = Ceil(None=VertexColor().G)
  n57 = ConstantBiasScale(None=@n16)
  n58 = LinearInterpolate(
   A=VectorParameter() {parameterName=TrackColor_01, defaultValue=<rgba 0.011 0.0176 0.0273 1>}.RGB
   B=VectorParameter() {parameterName=TrackColor_02, defaultValue=<rgba 0.0152 0.0222 0.0319 1>}.RGB
   Alpha=@n57
  )
  FunctionOutput(None=SetMaterialAttributes!(
    MaterialAttributes=@n17.MaterialAttributes
    EmissiveColor=LinearInterpolate(A=@n55, B=@n54, Alpha=@n56)
    BaseColor=LinearInterpolate(A=@n17.BaseColor, B=@n58, Alpha=@n56)
   ) {attributeSetTypes [#1, #2]})
  LinearInterpolate(
   A=@n20
   B=ScalarParameter() {defaultValue=0.5, parameterName=CoasterSpeedMax}
   Alpha=CollectionParameter() {collection=&MPC_JerkyCollection, parameterName=TrackStemSnare}
  )
  LinearInterpolate(B=Constant3Vector() {constant=<rgba 0.5 0.5 1 0>}, Alpha=@n27)
  LinearInterpolate(B=Constant3Vector() {constant=<rgba 0.1354 0.1354 0.1354 0>}, Alpha=@n27)
  n59 = TextureSampleParameter2D() {group=Planet, parameterName=Normal, samplerType=SAMPLERTYPE_Normal, texture=&FlatNormalBlue}
  MaterialFunctionCall!(Rotation Axis=@n8.RGB, Rotation Angle=@n9, World Space Vertex Normals=Transform(None=@n59.RGB), Movement Mask, PivotPoint) {materialFunction=&FixRotateAboutAxisNormals}
  n60 = TextureSampleParameter2D() {group=Planet, parameterName=Normal, samplerType=SAMPLERTYPE_Normal, texture=&FlatNormalBlue}
  MaterialFunctionCall(Rotation Axis=@n3.RGB, Rotation Angle=@n4, World Space Vertex Normals=Transform(None=@n60.RGB)) {materialFunction=&FixRotateAboutAxisNormals}
  n61 = TextureSampleParameter2D() {group=Planet, parameterName=Normal, samplerType=SAMPLERTYPE_Normal, texture=&FlatNormalBlue}
  MaterialFunctionCall(Rotation Axis=@n12.RGB, Rotation Angle=@n13, World Space Vertex Normals=Transform(None=@n61.RGB)) {materialFunction=&FixRotateAboutAxisNormals}
  n62 = Multiply(A=@n18.RGB, B=ScalarParameter() {defaultValue=0.1, parameterName=TrackEmissiveStrength})
  Multiply(A=@n56, B=LinearInterpolate(A=Multiply(A=@n62) {constB=0}, B=@n62, Alpha=@n57))
  Multiply(B=@n18.RGB)
  Normalize(VectorInput=VertexInterpolator(VS=PreSkinnedPosition()).PS)
  PreSkinnedPosition()
  VectorParameter() {parameterName=Spinner_01_Location, defaultValue=<rgba 0 0 0 1>}
  VectorParameter() {parameterName=Spinner_02_Location, defaultValue=<rgba 0 0 0 1>}
  VertexInterpolator(VS=PreSkinnedPosition())
  VertexInterpolator(VS=PreSkinnedPosition())
  WorldPosition()
 ]}
