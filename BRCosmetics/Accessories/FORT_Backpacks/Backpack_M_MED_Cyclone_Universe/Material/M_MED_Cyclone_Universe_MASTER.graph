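# Master material for the Cyclone Universe backpack cosmetic, built on material attributes for skeletal meshes.
# The shared character PreFX function feeds a backpack-specific V2 function, which also takes the SkinFX mask texture, defaulting to a planet FX map.
# A static switch, UseCycloneEmissives, optionally runs that result through a Cyclone emissives function before the shared PostFX function.
# The shading model comes from the material expression, using Default Lit and Subsurface Profile.
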Material /BRCosmetics/Accessories/FORT_Backpacks/Backpack_M_MED_Cyclone_Universe/Material/M_MED_Cyclone_Universe_MASTER
graph {
 properties {
  bUseMaterialAttributes=true
  bUsedWithMeshDeformer=true
  bUsedWithSkeletalMesh=true
  pixelDepthOffsetMode=PDOM_Legacy
  shadingModel=MSM_FromMaterialExpression
  usedShadingModels=Default Lit | Subsurface Profile
 }
 nodes [
  n1 = MaterialFunctionCall() {materialFunction=&BaseFN_CharacterFunction_PreFX}
  n2 = TextureSampleParameter2D() {
   parameterName=SkinFX
   samplerType=SAMPLERTYPE_Masks
   texture=/BRCosmetics/Accessories/FORT_Backpacks/Backpack_M_MED_Cyclone_Universe/Textures/T_M_MED_Cyclone_Universe_Backpack_Planet_FX
  }
  n3 = MaterialFunctionCall!(Gmap=@n1.Gmap, FX_Mask=@n2.RGB, Material=@n1.Gmap) {materialFunction=&/MF_CycloneUniverseBackpack_V2}
  MaterialFunctionCall(
   AO=@n1.AO
   Distance Blend=@n1.AO
   Fresnel=@n1.AO
   Gmap=@n1.AO
   Material=StaticSwitchParameter(
    True=MaterialFunctionCall!(Material=@n3.Result) {materialFunction=&/MF_CycloneUniverse_Emissives}.Result
    False=@n3.Result
   ) {parameterName=UseCycloneEmissives}
  ) {materialFunction=&BaseFN_CharacterFunction_PostFX}
 ]
}
