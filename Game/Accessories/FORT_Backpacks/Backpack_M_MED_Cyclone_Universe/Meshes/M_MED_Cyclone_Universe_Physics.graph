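# Physics asset for the Cyclone Universe back bling, giving its meteor and speaker box pieces simple collision and joints.
# It has three capsule bodies: a tiny one on the pack bone and larger ones on dyn_meteor and speaker_box.
# Both dyn_meteor and speaker_box are constrained directly to pack, forming two separate branches.
# The meteor joint locks twist, unlike the speaker box joint.
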PhysicsAsset /Game/Accessories/FORT_Backpacks/Backpack_M_MED_Cyclone_Universe/Meshes/M_MED_Cyclone_Universe_Physics
bodies {
 dyn_meteor [{shapeType=Capsule, length=1, radius=10.8834, center=<3.8871 0 0>}]
 pack [{shapeType=Capsule, length=0.505, radius=0.505, rotation=<rot 0 -90 0>}]
 speaker_box [{shapeType=Capsule, length=4.4697, radius=4.4571, center=<-0.0007 0.0003 0.0035>, rotation=<rot -3.4801 -90 -125.3584>}]
}
constraints [
 {bone1Name=speaker_box, bone2Name=pack, swing1LimitDegrees=45, swing2LimitDegrees=45, twistLimitDegrees=45}
 {bone1Name=dyn_meteor, bone2Name=pack, swing1LimitDegrees=45, swing2LimitDegrees=45, twistLimitDegrees=45, twistMotion=Locked}
]
