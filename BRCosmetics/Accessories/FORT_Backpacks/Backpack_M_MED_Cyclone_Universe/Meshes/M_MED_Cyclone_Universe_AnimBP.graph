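# Animation Blueprint for the Cyclone Universe backpack cosmetic that spins the dyn_meteor bone and shifts it while the owner jumps.
# Update events compute time-based spin, jump alpha and Perlin-noise reverb, while post-evaluate sends the meteor socket location to materials as Speaker_Location.
# Only the two meteor bone modifiers reach the output pose.
# Speaker reverb, pack scale and RigidBody nodes are disconnected, and the ferris-wheel/spinner material-parameter chain has no triggering event.
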AnimBlueprint /BRCosmetics/Accessories/FORT_Backpacks/Backpack_M_MED_Cyclone_Universe/Meshes/M_MED_Cyclone_Universe_AnimBP
target_skeleton=&/M_MED_Cyclone_Universe_Skeleton
graphs [
 AnimGraph {nodes [
   n1 = Get reverbTran() -> (reverbTran:Vector)
   n2 = Get reverbScale() -> (reverbScale:Vector)
   n3 = "Transform (Modify) Bone - Bone: speaker_front"(Alpha=1, Translation=@n1, Rotation="0,0,0", Scale=@n2) {
    rotationMode=BMM_Additive
    rotationSpace=BCS_BoneSpace
    scaleMode=BMM_Additive
    scaleSpace=BCS_BoneSpace
    translationMode=BMM_Additive
    translationSpace=BCS_BoneSpace
    alphaBoolBlend {blendOption=Linear}
    alphaScaleBiasClamp {inRange {max=25, min=10}}
    boneToModify {boneName=speaker_front}
   }
   n4 = "Transform (Modify) Bone - Bone: speaker_right"(ComponentPose=@n3, Alpha=1, Translation=@n1, Rotation="0,0,0", Scale=@n2) {
    rotationMode=BMM_Additive
    rotationSpace=BCS_BoneSpace
    scaleMode=BMM_Additive
    scaleSpace=BCS_BoneSpace
    translationMode=BMM_Additive
    translationSpace=BCS_BoneSpace
    alphaBoolBlend {blendOption=Linear}
    alphaScaleBiasClamp {inRange {max=25, min=10}}
    boneToModify {boneName=speaker_right}
   }
   n5 = "Transform (Modify) Bone - Bone: speaker_back"(ComponentPose=@n4, Alpha=1, Translation=@n1, Rotation="0,0,0", Scale=@n2) {
    rotationMode=BMM_Additive
    rotationSpace=BCS_BoneSpace
    scaleMode=BMM_Additive
    scaleSpace=BCS_BoneSpace
    translationMode=BMM_Additive
    translationSpace=BCS_BoneSpace
    alphaBoolBlend {blendOption=Linear}
    alphaScaleBiasClamp {inRange {max=25, min=10}}
    boneToModify {boneName=speaker_back}
   }
   n6 = "Transform (Modify) Bone - Bone: speaker_top"(ComponentPose=@n5, Alpha=1, Translation=@n1, Rotation="0,0,0", Scale=@n2) {
    rotationMode=BMM_Additive
    rotationSpace=BCS_BoneSpace
    scaleMode=BMM_Additive
    scaleSpace=BCS_BoneSpace
    translationMode=BMM_Additive
    translationSpace=BCS_BoneSpace
    alphaBoolBlend {blendOption=Linear}
    alphaScaleBiasClamp {inRange {max=25, min=10}}
    boneToModify {boneName=speaker_top}
   }
   n7 = "Transform (Modify) Bone - Bone: speaker_left"(ComponentPose=@n6, Alpha=1, Translation=@n1, Rotation="0,0,0", Scale=@n2) {
    rotationMode=BMM_Additive
    rotationSpace=BCS_BoneSpace
    scaleMode=BMM_Additive
    scaleSpace=BCS_BoneSpace
    translationMode=BMM_Additive
    translationSpace=BCS_BoneSpace
    alphaBoolBlend {blendOption=Linear}
    alphaScaleBiasClamp {inRange {max=25, min=10}}
    boneToModify {boneName=speaker_left}
   }
   "Transform (Modify) Bone - Bone: speaker_bot"(ComponentPose=@n7, Alpha=1, Translation=@n1, Rotation="0,0,0", Scale=@n2) {
    rotationMode=BMM_Additive
    rotationSpace=BCS_BoneSpace
    scaleMode=BMM_Additive
    scaleSpace=BCS_BoneSpace
    translationMode=BMM_Additive
    translationSpace=BCS_BoneSpace
    alphaBoolBlend {blendOption=Linear}
    alphaScaleBiasClamp {inRange {max=25, min=10}}
    boneToModify {boneName=speaker_bot}
   }
   "Transform (Modify) Bone - Bone: pack"(Translation="0,0,0", Rotation="0,0,0", Scale="0.75,0.75,0.75") {scaleMode=BMM_Additive, scaleSpace=BCS_BoneSpace, alphaBoolBlend {blendOption=Linear}, boneToModify {boneName=pack}, scale=<0.75 0.75 0.75>}
   RigidBody!(ComponentPose, Alpha=0.5) -> (Pose) {
    bClampLinearTranslationLimitToRefPose=true
    bOverrideWorldGravity=true
    overridePhysicsAsset=/Game/Accessories/FORT_Backpacks/Backpack_M_MED_Cyclone_Universe/Meshes/M_MED_Cyclone_Universe_Physics
    simulationSpace=BaseBoneSpace
    alphaBoolBlend {blendOption=Linear}
    baseBoneRef {boneName=root}
    componentAppliedLinearAccClamp=<3000 3000 3000>
    componentLinearAccScale=<0.25 0.25 0.25>
    componentLinearVelScale=<0.25 0.25 0.25>
    externalForce=<0.25 0.25 0.25>
    overrideWorldGravity {z=-800}
   }
   n8 = "Transform (Modify) Bone - Bone: dyn_meteor"(
    ComponentPose=Local To Component(LocalPose=Input Pose())
    Alpha=1
    Translation="0,0,0"
    Rotation=Get SpinnerRot() -> (SpinnerRot:Rotator)
    Scale="1,1,1"
   ) {rotationMode=BMM_Replace, rotationSpace=BCS_BoneSpace, alphaBoolBlend {blendOption=Linear}, boneToModify {boneName=dyn_meteor}}
   n9 = "Transform (Modify) Bone - Bone: dyn_meteor"(ComponentPose=@n8, Alpha=Get Jump_Alpha() -> (Jump_Alpha:double), Translation="-4,0,0", Rotation="0,0,0", Scale="1,1,1") {
    translationMode=BMM_Additive
    translationSpace=BCS_BoneSpace
    alphaBoolBlend {blendOption=Linear}
    alphaScaleBiasClamp {bInterpResult=true, bias=0.15, interpSpeedDecreasing=4, interpSpeedIncreasing=5, scale=0.75}
    boneToModify {boneName=dyn_meteor}
    translation {x=-4}
   }
   Output Pose(Result=Component To Local(ComponentPose=@n9)) {layerGroup=None}
  ]}
 EventGraph {nodes [
   n1 = GetPartSkeletalMeshComponent()
   SetVectorParameterValueOnMaterials!(
    execute
    self:Mesh Component Object Reference=[@n1, @n1]
    ParameterName:Name=FerisWheelLocalOffset_WPO
    ParameterValue:Vector=GetSocketLocation!(self:Scene Component Object Reference=@n1, InSocketName:Name=dyn_ferris_top) -> (ReturnValue:Vector)
   ) -> (then)
   > SetVectorParameterValueOnMaterials(self=@n1, ParameterName=Spinner_01_LocalOffset_WPO, ParameterValue=GetSocketLocation(self=@n1, InSocketName=dyn_spinnerA))
   > SetVectorParameterValueOnMaterials(self=@n1, ParameterName=Spinner_02_LocalOffset_WPO, ParameterValue=GetSocketLocation(self=@n1, InSocketName=dyn_spinnerB))
   > SetVectorParameterValueOnMaterials(
    self=@n1
    ParameterName=FerrisWheelRotationOrientation
    ParameterValue=GetUpVector!(InRot:Rotator=GetSocketRotation!(self:Scene Component Object Reference=@n1, InSocketName:Name=dyn_ferris_top) -> (ReturnValue:Rotator)) -> (ReturnValue:Vector)
   )
   > SetVectorParameterValueOnMaterials(
    self=@n1
    ParameterName=Spinner_A_RotationOrientation
    ParameterValue=GetUpVector(InRot=GetSocketRotation(self=@n1, InSocketName=dyn_spinnerA))
   )
   > SetVectorParameterValueOnMaterials(
    self=@n1
    ParameterName=Spinner_B_RotationOrientation
    ParameterValue=GetUpVector(InRot=GetSocketRotation(self=@n1, InSocketName=dyn_spinnerB))
   )
   Event BlueprintUpdateAnimation()
   > n2 = Sequence!(execute) -> (then_0, then_1, then_2:Exec)
   n3 = float / float(A=GetRealTimeSeconds(), B=-720) {is=title K2Node_CallFunction}
   > Set SpinnerRot(SpinnerRot:Rotator=MakeRotator(Roll=float * float(A=@n3, B=4500))) -> (Output_Get:Rotator)
   n4 = speakerReverb_FNC!(execute=@n2.then_2, self, frequenc:double=5, amp:double=0.3) -> (then, tran_output:double, scale_output:double)
   > Set reverbTran(reverbTran:Vector=MakeVector(Y=@n4)) -> (Output_Get:Vector)
   > Set reverbScale(reverbScale:Vector=MakeVector(X=@n4.scale_output, Y=@n4.scale_output, Z=@n4.scale_output)) -> (Output_Get:Vector)
   Set Jump_Alpha(execute=@n2.then_1, Jump_Alpha:double=SelectFloat(A=0.75, B=0, bPickA=Get bIsJumping())) -> (Output_Get:double)
   Event BlueprintPostEvaluateAnimation() {is=title K2Node_Event}
   > Is Valid(InputObject=@n1)
   > Delay(Duration=0)
   > SetVectorParameterValueOnMaterials(self=@n1, ParameterName=Speaker_Location, ParameterValue=GetSocketLocation(self=@n1, InSocketName=dyn_meteor))
   TryGetPawnOwner()
   TryGetPawnOwner()
  ]}
 speakerReverb_FNC {nodes [
   n1 = speakerReverb_FNC!() -> (then, frequenc:double=~, amp:double=~) {is=title K2Node_FunctionEntry}
   n2 = GetGameTimeInSeconds()
   n3 = "Sin (Radians)"(A=PerlinNoise1D!(Value:float=float * float(A=@n1, B=@n2)) -> (ReturnValue:float))
   > Return Node!(
    execute
    tran_output:double=PerlinNoise1D(Value=float * float(A=@n3, B=@n1.amp))
    scale_output:double=Make Perlin Noise and Remap!(Value:float=@n3, RangeOutMin:float=0.9, RangeOutMax:float=1.1) -> (ReturnValue:float)
   )
  ]}
]
variables [
 Jump_Alpha {type=number}
 reverbScale {type=Vector}
 reverbTran {type=Vector}
 SpinnerRot {type=Rotator}
]
